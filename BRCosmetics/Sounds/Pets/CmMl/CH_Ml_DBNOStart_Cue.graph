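# SoundCue for the CmMl pet cosmetic's DBNO (down but not out) start sound.
# A quality-level node sends the first two tiers to the standard wave and the last two to an _opt variant, each through a pitch-randomizing Modulator.
# It uses close-range pet attenuation and the creature pets sound class, with overridden concurrency allowing one instance per owner and preventing new plays.
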SoundCue /BRCosmetics/Sounds/Pets/CmMl/CH_Ml_DBNOStart_Cue
root=QualityLevel_0
nodes [
 {id=Modulator_0, children [WavePlayer_1], properties {pitchMax=1.2}}
 {id=Modulator_1, children [WavePlayer_0], properties {pitchMax=1.2}}
 {id=QualityLevel_0, children [Modulator_0,Modulator_0,Modulator_1,Modulator_1]}
 {id=WavePlayer_0, properties {soundWaveAssetPtr=&/CH_Ml_DBNOStart_01_opt}}
 {id=WavePlayer_1, properties {soundWaveAssetPtr=&/CH_Ml_DBNOStart_01}}
]
properties {
 attenuationSettings=&Fort_Pets_Close_Athena_Att
 bOverrideConcurrency=true
 maxDistance=1800.0001
 soundClassObject=&Fort_Creatures_Pets
 volumeMultiplier=1.12
 concurrencyOverrides {bLimitToOwner=true, maxCount=1, resolutionRule=PreventNew}
 concurrencySet [/Game/Sounds/Concurrency_Rules/ThreeMaximum_Owner_PreventNew]
}
