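# Sound cue for the happy vocalization of the CmMl cosmetic pet, routed to the creature pets sound class with close-range pet attenuation.
# A quality-level switch feeds an equal-weight random pick of four waves on the first two levels.
# The other two levels share a single "_opt" variant of the first wave.
# A concurrency override allows one instance per owner and blocks new plays while it is active.
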SoundCue /BRCosmetics/Sounds/Pets/CmMl/PT_Cm_Ml_Happy_Cue
root=QualityLevel_0
nodes [
 {id=QualityLevel_0, children [Random_1,Random_1,WavePlayer_1,WavePlayer_1]}
 {id=Random_1, children [WavePlayer_0,WavePlayer_4,WavePlayer_5,WavePlayer_6], properties {weights [1,1,1,1]}}
 {id=WavePlayer_0, properties {soundWaveAssetPtr=&/PT_Cm_Ml_Happy_01}}
 {id=WavePlayer_1, properties {soundWaveAssetPtr=&/PT_Cm_Ml_Happy_01_opt}}
 {id=WavePlayer_4, properties {soundWaveAssetPtr=&/PT_Cm_Ml_Happy_02}}
 {id=WavePlayer_5, properties {soundWaveAssetPtr=&/PT_Cm_Ml_Happy_03}}
 {id=WavePlayer_6, properties {soundWaveAssetPtr=&/PT_Cm_Ml_Happy_04}}
]
properties {
 attenuationSettings=&Fort_Pets_Close_Athena_Att
 bOverrideConcurrency=true
 maxDistance=1800.0001
 soundClassObject=&Fort_Creatures_Pets
 volumeMultiplier=1.12
 concurrencyOverrides {bLimitToOwner=true, maxCount=1, resolutionRule=PreventNew}
 concurrencySet [/Game/Sounds/Concurrency_Rules/ThreeMaximum_Owner_PreventNew]
}
